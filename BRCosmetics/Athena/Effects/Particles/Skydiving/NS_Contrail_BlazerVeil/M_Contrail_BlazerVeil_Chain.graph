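# Masked, two-sided Default Lit material for Niagara ribbons that renders the chain in the BlazerVeil skydiving contrail.
# One packed texture, with scalable U tiling, supplies normal XY in red-green (Z derived), shading in blue and mask in alpha.
# Blue tints Base Color, which feeds emissive via particle color, an intensity parameter and a shared adjustment function.
# A parameter blends particle alpha into the mask; scalar parameters expose roughness, metallic and specular.
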Material /BRCosmetics/Athena/Effects/Particles/Skydiving/NS_Contrail_BlazerVeil/M_Contrail_BlazerVeil_Chain
graph {
 properties {bUsedWithNiagaraRibbons=true, blendMode=BLEND_Masked, pixelDepthOffsetMode=PDOM_Legacy, twoSided=true, usedShadingModels=Default Lit}
 nodes [
  n1 = TextureCoordinate()
  n2 = AppendVector(A=ScalarParameter() {defaultValue=1.5, parameterName=UV Scale}, B=Constant() {r=1})
  n3 = TextureSampleParameter2D(UVs=Add(A=Multiply(A=@n1, B=@n2))) {parameterName=Chain Texture, texture=&/T_Contrail_BlazerVeil_CubanChain}
  DeriveNormalZ(InXY=ConstantBiasScale(None=ComponentMask(None=@n3.RGBA) {g=true, r=true}) {bias=-0.5, scale=2})
  n4 = ComponentMask(None=@n3.RGBA) {a=true}
  n5 = ParticleColor()
  LinearInterpolate(A=@n4, B=Multiply(A=@n5.A, B=@n4), Alpha=ScalarParameter() {parameterName=Use ParticleAlpha?})
  n6 = Multiply(
   A=ComponentMask(None=@n3.RGBA) {b=true}
   B=VectorParameter() {group=Appearance, parameterName=Base Color, defaultValue=<rgba 0.6927 0.3668 0.2706 0.55>}.RGB
  )
  n7 = ScalarParameter() {group=Appearance, parameterName=Emissive}
  MaterialFunctionCall(Emissive=Multiply(A=@n5.RGB, B=Multiply(A=@n6, B=@n7))) {materialFunction=&MF_FX_EmissiveAdjustments}
  ScalarParameter() {group=Appearance, parameterName=Roughness, sliderMax=1}
  ScalarParameter() {defaultValue=1, group=Appearance, parameterName=Metallic, sliderMax=1}
  ScalarParameter() {defaultValue=1, group=Appearance, parameterName=Specular, sliderMax=1}
 ]
}
