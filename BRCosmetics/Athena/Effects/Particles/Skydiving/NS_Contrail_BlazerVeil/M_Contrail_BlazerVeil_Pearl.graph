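# Lit material for the pearl mesh particles in the Blazer Veil skydiving contrail Niagara effect.
# A Base Color parameter multiplied by particle color forms the tinted surface color, alongside scalar parameters for Metallic, Specular and Roughness.
# The same color, scaled by an Emissive parameter and particle color, feeds the MF_FX_EmissiveAdjustments function for glow.
# It samples no textures, so parameters and particle color alone control the look.
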Material /BRCosmetics/Athena/Effects/Particles/Skydiving/NS_Contrail_BlazerVeil/M_Contrail_BlazerVeil_Pearl
graph {
 properties {bUsedWithNiagaraMeshParticles=true, pixelDepthOffsetMode=PDOM_Legacy, usedShadingModels=Default Lit}
 nodes [
  n1 = VectorParameter() {parameterName=Base Color}
  n2 = ScalarParameter() {parameterName=Emissive}
  n3 = ParticleColor()
  MaterialFunctionCall(Emissive=Multiply(A=Multiply(A=@n1.RGB, B=@n2), B=@n3.RGB)) {materialFunction=&MF_FX_EmissiveAdjustments}
  Multiply(A=@n1.RGB, B=@n3.RGB)
  ScalarParameter() {parameterName=Metallic, sliderMax=1}
  ScalarParameter() {parameterName=Specular, sliderMax=1}
  ScalarParameter() {parameterName=Roughness, sliderMax=1}
 ]
}
