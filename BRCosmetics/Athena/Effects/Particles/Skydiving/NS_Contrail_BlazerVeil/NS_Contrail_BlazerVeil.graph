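# Skydiving contrail for the BlazerVeil cosmetic: jeweled ribbon chains hung from sampled skeletal mesh points, with diamonds, pearls and sparkles.
# Bracelet chains ring the points, strands trail against owner velocity with sine sway, and web chains follow Hermite splines.
# Cross-web chains read web positions and sag through pendulum constraints.
# Gem and sparkle emitters use particle attribute readers to ride or shed from chains, with time-of-day dimming on most emitters.
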NiagaraSystem /BRCosmetics/Athena/Effects/Particles/Skydiving/NS_Contrail_BlazerVeil/NS_Contrail_BlazerVeil
dependencies {
 usedDataInterfaces [
  &NiagaraDataInterfaceCurve
  &NiagaraDataInterfaceParticleRead
  &NiagaraDataInterfaceArrayInt32
  &NiagaraDataInterfaceArrayFloat
  &NiagaraDataInterfaceArrayFloat3
 ]
 usedDynamicInputs []
 usedModules [
  &self:SystemSpawnScript.SystemScriptSource.SystemScriptGraph.NiagaraNodeAssignment_2.SetVariables_#1
  &SystemState
  /Game/Effects/Niagara/Modules/Skydiving/FN_DivingAttributes/v4/FN_DivingAttributes_v4
  &self:SystemSpawnScript.SystemScriptSource.SystemScriptGraph.NiagaraNodeAssignment_0.SetVariables_#2
 ]
}
emitters {
 Bracelet_Chains {
  emitterSpawnScript [&FN_PreserveSkydiveUserVariables() {module=PreserveUserVariables}]
  emitterUpdateScript [
   &EmitterState(
    Inactive Response=::NewEnumerator1
    Life Cycle Mode=::NewEnumerator1
    Loop Behavior=::NewEnumerator1
    Loop Duration Mode=::NewEnumerator1
    UseLoopDelay=false
   )
   &SpawnBurst_Instantaneous(Spawn Count=20, Spawn Probability=HlslExpression() {hlslExpression="Emitter.LoopCount == 0"}, Use Spawn Probability=false) {module=SpawnBurst_Instantaneous001}
  ]
  particleSpawnScript [
   &FN_SetSpawnGroup(Inclusivity=true, Normalize Spawn Group=true, Particles Per Group=10, Pattern=::NewEnumerator0, Set Ribbon ID and Ribbon Link Order=true)
   &InitializeParticle(Ribbon Width=3, Ribbon Width Mode=::NewEnumerator1, UsePositionOffset=false, Write Material Random=true) {module=InitializeParticle001}
   &DynamicMaterialParameters(
    Index 0 Param 1=&SelectFloatFromArray(
     Direct Array Index=&SpawnGroup()
     Float Selection Array=DataInterface() {propertyValues {FloatData [1,3,0.5,1.68,2.3,6.4]}}
    )
    Module.Param0WriteEnabled=true
    Module.Param1WriteEnabled=true
    Module.Param2WriteEnabled=true
    Module.Param3WriteEnabled=true
   )
  ]
  particleUpdateScript [
   &ParticleState(Kill Particles When Lifetime Has Elapsed=false)
   &SkeletalMeshLocation(
    Bone / Socket Index=&ModuloInt(A=&SpawnGroup(), B=2)
    Bone Sampling Mode=::NewEnumerator4
    MaskSpawnGroup=false
    Orientation Sampling=::NewEnumerator1
    Sampled Position Offset=/Niagara/DynamicInputs/Vectors/RotateVector(
     Vector To Rotate:Vector3f=&Multiply_VectorByFloat(
      Vector=&NormalizeVector(Vector To Normalize=<0 0 1>, Safe Normalize=true, Safe Fallback Threshold=0, Safe Fallback Vector=<0 0 0.001>)
      Float=5
     )
     Yaw:NiagaraFloat=0.4
     Pitch:NiagaraFloat=0
     Roll:NiagaraFloat=&Add_Float(
      A=$Particles.Normalized Spawn Group
      B=&SelectFloatFromArray(
       Direct Array Index=&SpawnGroup()
       Float Selection Array=DataInterface() {propertyValues {FloatData [0.1,0.5,0.6,0.27,0.86,0.68]}}
      )
     )
     Rotation Rate:NiagaraFloat=360
    )
    Skeletal Mesh=$User.SkeletalMesh
    Use Sampled Mesh Attribute as Ribbon ID=false
   )
   &DynamicMaterialParameters(
    Index 1 Param 1=&MakeFloatFromVector(Vector=$Particles.SkeletalMeshLocation.SampledPosition, Channel=::NewEnumerator0)
    Index 1 Param 2=&MakeFloatFromVector(Vector=$Particles.SkeletalMeshLocation.SampledPosition, Channel=::NewEnumerator1)
    Index 1 Param 3=&MakeFloatFromVector(Vector=$Particles.SkeletalMeshLocation.SampledPosition, Channel=::NewEnumerator2)
    Module.Index1Param0Write=true
    Module.Index1Param1Write=true
    Module.Index1Param2Write=true
    Module.Index1Param3Write=true
    Write Parameter Index 0=false
    Write Parameter Index 1=true
   ) {module=DynamicMaterialParameters001}
   &PendulumSetup(Module.Enable Spring Driver=false, Pendulum Length=0.5, Pendulum Pivot:NiagaraPosition=$Particles.Position)
   &SolveForcesAndVelocity(Clamp Velocity=false, Limit Acceleration=false)
   &PendulumConstraint()
   &CalculateAccurateVelocity()
   &ScaleColor(
    Scale RGB=&VectorFromFloat(Value=&FNTimeOfDayScalar(Night Scale=0.8, Morning Scale=0.9, Evening Scale=0.9))
    ScaleA=true
    ScaleRGB=true
   )
  ]
  rendererClasses [&NiagaraRibbonRendererProperties]
  renderers [{rendererClass=&NiagaraRibbonRendererProperties}]
 }
 Cross_Web_Chains {
  emitterSpawnScript [&FN_PreserveSkydiveUserVariables()]
  emitterUpdateScript [
   &EmitterState(
    Inactive Response=::NewEnumerator1
    Life Cycle Mode=::NewEnumerator1
    Loop Behavior=::NewEnumerator1
    Loop Duration Mode=::NewEnumerator1
    UseLoopDelay=false
   )
   /SetVariables(Emitter.Number of Chains=15, Emitter.Particle per Chain=6) {isSetParameters=true}
   &SpawnBurst_Instantaneous(
    Spawn Count=&Multiply_Int(A=$Emitter.Number of Chains, B=$Emitter.Particle per Chain)
    Use Spawn Probability=false
    UseLoopCountLimit=false
   )
  ]
  particleSpawnScript [
   &FN_SetSpawnGroup(
    Inclusivity=true
    Normalize Spawn Group=true
    Particles Per Group:NiagaraInt32=$Emitter.Particle per Chain
    Pattern=::NewEnumerator0
    Set Ribbon ID and Ribbon Link Order=true
   )
   &InitializeRibbon(Ribbon Width=3, UsePositionOffset=false, Write Material Random=true)
   &DynamicMaterialParameters(
    Index 0 Param 1=0.125
    Module.Param0WriteEnabled=true
    Module.Param1WriteEnabled=true
    Module.Param2WriteEnabled=false
    Module.Param3WriteEnabled=false
   )
  ]
  particleUpdateScript [
   &ParticleState(Kill Particles When Lifetime Has Elapsed=false, Let Infinitely Lived Particles Die When Emitter Deactivates=true)
   &self:ReadWebPosition(
    New Particle Attribute Reader=DataInterface() {propertyValues {EmitterBinding {bindingMode=Other, emitterName=Web_Chains}}}
    Number of Chains:NiagaraInt32=$Emitter.Number of Chains
    Spawn Group:NiagaraInt32=&SpawnGroup()
   ) {module=ReadWebPosition}
   /SetVariables(Particles.Origin:Vector3f=&Lerp_Vector(A=$Particles.StartSampledPosition, B=$Particles.StopSampledPosition, Alpha=$Particles.Normalized Spawn Group)) {isSetParameters=true}
   &PendulumSetup(
    Module.Enable Spring Driver=false
    Pendulum Length:NiagaraFloat=&FloatFromCurve(
     FloatCurve=DataInterface() {propertyValues {Curve {keys [
         {arriveTangent=3.5081, interpMode=RCIM_Cubic, leaveTangent=3.5081, leaveTangentWeight=0.5867, tangentMode=RCTM_User, time=0, value=0}
         {
          arriveTangent=-0.0344
          arriveTangentWeight=0.3375
          interpMode=RCIM_Cubic
          leaveTangent=-0.0344
          leaveTangentWeight=0.3375
          tangentMode=RCTM_User
          time=0.5
          value=1
         }
         {arriveTangent=-3.542, arriveTangentWeight=0.617, interpMode=RCIM_Cubic, leaveTangent=-3.542, tangentMode=RCTM_User, time=1, value=0}
        ]}}}
     CurveIndex=$Particles.Normalized Spawn Group
     Scale Curve=5
    )
    Pendulum Pivot:Vector3f=$Particles.Origin
    Rigid Pendulum=false
   )
   &GravityForce(Gravity=<0 0 980>)
   &SolveForcesAndVelocity(Clamp Velocity=false, Limit Acceleration=false)
   &PendulumConstraint(Calculate and Output Potential Energy=true)
   &CalculateAccurateVelocity()
   &ScaleRibbonWidth(Ribbon Width Scale=&FloatFromCurve(
     FloatCurve=DataInterface() {propertyValues {Curve {keys [
         {
          arriveTangent=1.3333
          arriveTangentWeight=0.0667
          interpMode=RCIM_Cubic
          leaveTangent=1.3333
          leaveTangentWeight=0.3432
          tangentMode=RCTM_User
          time=0
          value=0.725
         }
         {arriveTangentWeight=0.1667, interpMode=RCIM_Cubic, leaveTangentWeight=0.1667, tangentMode=RCTM_User, time=0.5, value=1}
         {arriveTangent=-0.8835, arriveTangentWeight=0.3432, interpMode=RCIM_Cubic, leaveTangent=-0.8835, tangentMode=RCTM_User, time=1, value=0.725}
        ]}}}
     CurveIndex=$Particles.Normalized Spawn Group
    ))
   &ScaleColor(
    Scale RGB=&VectorFromFloat(Value=&FNTimeOfDayScalar(Night Scale=0.8, Morning Scale=0.9, Evening Scale=0.9))
    ScaleA=true
    ScaleRGB=true
   ) {module=ScaleColor001}
  ]
  rendererClasses [&NiagaraRibbonRendererProperties]
  renderers [{rendererClass=&NiagaraRibbonRendererProperties}]
 }
 Diamonds_On_Strand_Chains {
  emitterUpdateScript [
   &EmitterState(Life Cycle Mode=::NewEnumerator1, Loop Behavior=::NewEnumerator1, Loop Duration=2, UseLoopDelay=false)
   &SpawnBurst_Instantaneous(Spawn Count=6, Use Spawn Probability=false, UseLoopCountLimit=false)
  ]
  particleSpawnScript [
   &InitializeParticle(
    Color=<rgba 10 5.282 3.3854 1>
    Lifetime Max=6
    Lifetime Min=3
    Lifetime Mode=::NewEnumerator1
    Mesh Scale Mode=::NewEnumerator3
    Mesh Uniform Scale=0.15
    Sprite Rotation Mode=::NewEnumerator2
    Sprite Size Mode=::NewEnumerator1
    Sprite UV Mode=::NewEnumerator3
    Uniform Sprite Size Max=20
    Uniform Sprite Size Min=15
    UsePositionOffset=false
    Write Material Random=true
   )
   &CameraOffset(Camera Offset Amount=5)
   &InitialMeshOrientation(Mesh Axis To Orient=<0 0 1>, Mesh Orientation Mode=::NewEnumerator2, Use Rotation Vector=false)
  ]
  particleUpdateScript [
   &ParticleState(Kill Particles When Lifetime Has Elapsed=false, Loop Particles Lifetime=true)
   &self:ScratchModule(
    New Particle Attribute Reader=DataInterface() {propertyValues {EmitterBinding {bindingMode=Other, emitterName=Strand_Chains}}}
    Particle Index:NiagaraInt32=&SelectIntFromArray(
     Direct Array Index=&ReturnExecIndex()
     Int Selection Array=DataInterface() {propertyValues {IntData [19,39,59,79,99,119]}}
    )
   ) {module=ScratchModule}
   &InterpolateOverTime(
    Rate Of Change=100
    Target Vector Value:Vector3f=&Multiply_Vector(A=$Output.ScratchModule.Direction Vector, B=<1 1 0>)
    Variable Type=::NewEnumerator2
   )
   &UpdateMeshOrientation(
    Facing Direction=&Lerp_Vector(A=$Particles.InterpolateOverTime.InterpolateOverTime_V3_.Moving Average, B=<0 0 1>, Alpha=0.6)
    Facing Interpolation=false
    Orientation Method=::NewEnumerator1
    Override Quat=false
    TransformFacing=false
    TransformUp=false
   )
   &SpriteRotationRate(Rotation Rate=&RandomRangeFloat(Minimum=200, Maximum=360))
   &ScaleSpriteSize(
    Uniform Curve Index=&ModuloFloat(
     A=&Add_Float(A=$Particles.NormalizedAge, B=&RandomRangeFloat(Minimum=0, Maximum=1))
     B=1
    )
    Uniform Curve Sprite Scale=DataInterface() {propertyValues {Curve {keys [
        {arriveTangent=10.2355, interpMode=RCIM_Cubic, leaveTangent=10.2355, tangentMode=RCTM_User, time=0, value=0}
        {arriveTangent=0.0075, interpMode=RCIM_Cubic, leaveTangent=0.0075, tangentMode=RCTM_User, time=0.0666, value=1}
        {arriveTangent=-0.0405, interpMode=RCIM_Cubic, leaveTangent=-0.0405, tangentMode=RCTM_User, time=0.15, value=0}
       ]}}}
   )
   &FN_Sprite_Attributes(
    Camera Offset:NiagaraFloat=&Add_Float(A=$Particles.CameraOffset, B=5)
    Create Camera Offset Attribute=true
    Create Position Attribute=true
    Position:NiagaraPosition=&AddVectorToPosition(Position=$Particles.Position, Vector=<1 0 3>)
    Write to Module Namespace=true
   )
  ]
  rendererClasses [&NiagaraMeshRendererProperties, &NiagaraSpriteRendererProperties]
  renderers [
   {rendererClass=&NiagaraMeshRendererProperties, rendererIndex=0}
   {rendererClass=&NiagaraSpriteRendererProperties, rendererIndex=1}
  ]
 }
 Hand_And_Center_Diamond {
  emitterUpdateScript [
   &EmitterState(Life Cycle Mode=::NewEnumerator1, Loop Behavior=::NewEnumerator1, Loop Duration=2, UseLoopDelay=false)
   &SpawnBurst_Instantaneous(Spawn Count=3, Use Spawn Probability=false, UseLoopCountLimit=false)
  ]
  particleSpawnScript [
   &InitializeParticle(
    Color=<rgba 10 5.282 3.3854 1>
    Lifetime Max=5
    Lifetime Min=3
    Lifetime Mode=::NewEnumerator1
    Mesh Scale Mode=::NewEnumerator3
    Mesh Uniform Scale=0.15
    Sprite Rotation Mode=::NewEnumerator2
    Sprite Size Mode=::NewEnumerator1
    Sprite UV Mode=::NewEnumerator3
    Uniform Sprite Size Max=20
    Uniform Sprite Size Min=15
    UsePositionOffset=false
    Write Material Random=true
   )
   &CameraOffset(Camera Offset Amount=5)
   &InitialMeshOrientation(Mesh Axis To Orient=<0 0 1>, Mesh Orientation Mode=::NewEnumerator2, Use Rotation Vector=false)
  ]
  particleUpdateScript [
   &ParticleState(Kill Particles When Lifetime Has Elapsed=false, Loop Particles Lifetime=true)
   &self:ScratchModule(
    New Particle Attribute Reader=DataInterface() {propertyValues {EmitterBinding {bindingMode=Other, emitterName=Web_Chains}}}
    Particle Index:NiagaraInt32=&SelectIntFromArray(
     Direct Array Index=&ReturnExecIndex()
     Int Selection Array=DataInterface() {propertyValues {IntData [0,14,22]}}
    )
   ) {module=ScratchModule}
   &InterpolateOverTime(
    Rate Of Change=5
    Target Vector Value:Vector3f=&Subtract_Position(A=$Particles.Position, B=$Particles.Previous.Position)
    Variable Type=::NewEnumerator2
   )
   &UpdateMeshOrientation(
    Facing Direction=&Lerp_Vector(
     A=&Multiply_VectorByFloat(
      Vector=&NormalizeVector(
       Vector To Normalize=&FN_FrontEndVector3Modifier(
        Gameplay=&TransformVector(Vector=$Engine.Owner.Velocity, Source Space=::World, Destination Space=::Simulation)
        Front End=$System.FrontendVelocity
       )
       Safe Normalize=true
       Safe Fallback Threshold=0
       Safe Fallback Vector=<0 0 0.001>
      )
      Float=-1
     )
     B=$Particles.InterpolateOverTime.InterpolateOverTime_V3_.Moving Average
     Alpha=0.5
    )
    Facing Interpolation=false
    Orientation Method=::NewEnumerator1
    Override Quat=false
    TransformFacing=false
    TransformUp=false
   )
   &SpriteRotationRate(Rotation Rate=&RandomRangeFloat(Minimum=200, Maximum=360))
   &ScaleSpriteSize(
    Uniform Curve Index=&ModuloFloat(
     A=&Add_Float(A=$Particles.NormalizedAge, B=&RandomRangeFloat(Minimum=0, Maximum=1))
     B=1
    )
    Uniform Curve Sprite Scale=DataInterface() {propertyValues {Curve {keys [
        {arriveTangent=10.2355, interpMode=RCIM_Cubic, leaveTangent=10.2355, tangentMode=RCTM_User, time=0, value=0}
        {arriveTangent=0.0075, interpMode=RCIM_Cubic, leaveTangent=0.0075, tangentMode=RCTM_User, time=0.0666, value=1}
        {arriveTangent=-0.0405, interpMode=RCIM_Cubic, leaveTangent=-0.0405, tangentMode=RCTM_User, time=0.15, value=0}
       ]}}}
   )
   &FN_Sprite_Attributes(
    Camera Offset:NiagaraFloat=&Add_Float(A=$Particles.CameraOffset, B=5)
    Create Camera Offset Attribute=true
    Create Position Attribute=true
    Position:NiagaraPosition=&AddVectorToPosition(Position=$Particles.Position, Vector=<1 0 4>)
    Write to Module Namespace=true
   )
  ]
  rendererClasses [&NiagaraMeshRendererProperties, &NiagaraSpriteRendererProperties]
  renderers [
   {rendererClass=&NiagaraMeshRendererProperties, rendererIndex=0}
   {rendererClass=&NiagaraSpriteRendererProperties, rendererIndex=1}
  ]
 }
 Pearls_On_Strand_Chains {
  emitterUpdateScript [
   &EmitterState(Life Cycle Mode=::NewEnumerator1, Loop Behavior=::NewEnumerator1, Loop Duration=2, UseLoopDelay=false)
   &SpawnBurst_Instantaneous(Spawn Count=18, Use Spawn Probability=false, UseLoopCountLimit=false)
  ]
  particleSpawnScript [
   &InitializeParticle(
    Color=&MakeCustomLinearColorFromBool(
     Bool=&Boolean_OrOperation(Bool1=$User.bIsFrontEnd, Bool2=$User.bIsFrontEndPreview)
     True Color=$User.PearlColorFrontend
     False Color=$User.PearlColor
    )
    Lifetime Max=6
    Lifetime Min=3
    Lifetime Mode=::NewEnumerator1
    Mesh Scale Mode=::NewEnumerator3
    Mesh Uniform Scale=&Multiply_Float(A=10, B=0.001)
    Sprite Rotation Mode=::NewEnumerator2
    Sprite Size Mode=::NewEnumerator1
    Sprite UV Mode=::NewEnumerator3
    Uniform Sprite Size Max=15
    Uniform Sprite Size Min=10
    UsePositionOffset=false
    Write Material Random=true
   )
   &CameraOffset(Camera Offset Amount=5)
  ]
  particleUpdateScript [
   &ParticleState(Kill Particles When Lifetime Has Elapsed=false, Loop Particles Lifetime=true)
   &self:ScratchModule(
    New Particle Attribute Reader=DataInterface() {propertyValues {EmitterBinding {bindingMode=Other, emitterName=Strand_Chains}}}
    Particle Index:NiagaraInt32=&SelectIntFromArray(
     Direct Array Index=&ReturnExecIndex()
     Int Selection Array=DataInterface() {propertyValues {IntData [13,15,18,33,35,38,53,55,58,73,75,78,93,95,98,113,115,118]}}
    )
   ) {module=ScratchModule}
   &SpriteRotationRate(Rotation Rate=&RandomRangeFloat(Minimum=200, Maximum=360))
   &ScaleSpriteSize(
    Uniform Curve Index=&ModuloFloat(
     A=&Add_Float(A=$Particles.NormalizedAge, B=&RandomRangeFloat(Minimum=0, Maximum=1))
     B=1
    )
    Uniform Curve Sprite Scale=DataInterface() {propertyValues {Curve {keys [
        {arriveTangent=10.2355, interpMode=RCIM_Cubic, leaveTangent=10.2355, tangentMode=RCTM_User, time=0, value=0}
        {arriveTangent=0.0075, interpMode=RCIM_Cubic, leaveTangent=0.0075, tangentMode=RCTM_User, time=0.0666, value=1}
        {arriveTangent=-0.0405, interpMode=RCIM_Cubic, leaveTangent=-0.0405, tangentMode=RCTM_User, time=0.15, value=0}
       ]}}}
   )
   &FN_Sprite_Attributes(
    Camera Offset:NiagaraFloat=&Add_Float(A=$Particles.CameraOffset, B=5)
    Create Camera Offset Attribute=true
    Write to Module Namespace=true
   )
   &ScaleColor(
    Scale RGB=&VectorFromFloat(Value=&FNTimeOfDayScalar(Night Scale=0.5, Morning Scale=0.75, Evening Scale=0.75))
    ScaleA=true
    ScaleRGB=true
   )
  ]
  rendererClasses [&NiagaraMeshRendererProperties, &NiagaraSpriteRendererProperties]
  renderers [
   {rendererClass=&NiagaraMeshRendererProperties, rendererIndex=0}
   {rendererClass=&NiagaraSpriteRendererProperties, rendererIndex=1}
  ]
 }
 Pearls_On_Web_Chains {
  emitterUpdateScript [
   &EmitterState(Life Cycle Mode=::NewEnumerator1, Loop Behavior=::NewEnumerator1, Loop Duration Mode=::NewEnumerator1, UseLoopDelay=false)
   &SpawnBurst_Instantaneous(Spawn Count=10, Use Spawn Probability=false, UseLoopCountLimit=false)
  ]
  particleSpawnScript [
   &InitializeParticle(
    Color=&MakeCustomLinearColorFromBool(
     Bool=&Boolean_OrOperation(Bool1=$User.bIsFrontEnd, Bool2=$User.bIsFrontEndPreview)
     True Color=$User.PearlColorFrontend
     False Color=$User.PearlColor
    )
    Lifetime Max=6
    Lifetime Min=3
    Lifetime Mode=::NewEnumerator1
    Mesh Scale Mode=::NewEnumerator3
    Mesh Uniform Scale=&Multiply_Float(A=9, B=0.001)
    Sprite Rotation Mode=::NewEnumerator2
    Sprite Size Mode=::NewEnumerator1
    Sprite UV Mode=::NewEnumerator3
    Uniform Sprite Size Max=15
    Uniform Sprite Size Min=10
    UsePositionOffset=false
    Write Material Random=true
   )
   &CameraOffset(Camera Offset Amount=5)
  ]
  particleUpdateScript [
   &ParticleState(Kill Particles When Lifetime Has Elapsed=false, Loop Particles Lifetime=true)
   &self:ScratchModule(
    New Particle Attribute Reader=DataInterface() {propertyValues {EmitterBinding {bindingMode=Other, emitterName=Web_Chains}}}
    Particle Index:NiagaraInt32=&SelectIntFromArray(
     Direct Array Index=&ReturnExecIndex()
     Direct Array Mode=::NewEnumerator1
     Int Selection Array=DataInterface() {propertyValues {IntData [3,5,7,9,11,18,20,22,24,26]}}
    )
   ) {module=ScratchModule}
   &SpriteRotationRate(Rotation Rate=&RandomRangeFloat(Minimum=200, Maximum=360))
   &ScaleSpriteSize(
    Uniform Curve Index=&ModuloFloat(
     A=&Add_Float(A=$Particles.NormalizedAge, B=&RandomRangeFloat(Minimum=0, Maximum=1))
     B=1
    )
    Uniform Curve Sprite Scale=DataInterface() {propertyValues {Curve {keys [
        {arriveTangent=10.2355, interpMode=RCIM_Cubic, leaveTangent=10.2355, tangentMode=RCTM_User, time=0, value=0}
        {arriveTangent=0.0075, interpMode=RCIM_Cubic, leaveTangent=0.0075, tangentMode=RCTM_User, time=0.0666, value=1}
        {arriveTangent=-0.0405, interpMode=RCIM_Cubic, leaveTangent=-0.0405, tangentMode=RCTM_User, time=0.15, value=0}
       ]}}}
   )
   &FN_Sprite_Attributes(
    Camera Offset:NiagaraFloat=&Add_Float(A=$Particles.CameraOffset, B=5)
    Create Camera Offset Attribute=true
    Write to Module Namespace=true
   )
   &ScaleColor(
    Scale RGB=&VectorFromFloat(Value=&FNTimeOfDayScalar(Night Scale=0.5, Morning Scale=0.75, Evening Scale=0.75))
    ScaleA=true
    ScaleRGB=true
   )
  ]
  rendererClasses [&NiagaraMeshRendererProperties, &NiagaraSpriteRendererProperties]
  renderers [
   {rendererClass=&NiagaraMeshRendererProperties, rendererIndex=0}
   {rendererClass=&NiagaraSpriteRendererProperties, rendererIndex=1}
  ]
 }
 Sparkles_Everywhere_Flying {
  emitterUpdateScript [
   &EmitterState(Life Cycle Mode=::NewEnumerator1, Loop Behavior=::NewEnumerator1, Loop Duration Mode=::NewEnumerator1, UseLoopDelay=false)
   &SpawnRate(SpawnRate=&Lerp_Float(A=10, B=20, Alpha=$User.ParaGlideLeanAlpha), Use Spawn Probability=false)
  ]
  particleSpawnScript [
   &InitializeParticle(
    Color=&RandomRangeLinearColor(Minimum=<rgba 10 5.4167 5 1>, Maximum=<rgba 10 7 4 1>)
    Lifetime Max=1
    Lifetime Min=0.25
    Lifetime Mode=::NewEnumerator1
    Sprite Rotation Mode=::NewEnumerator2
    Sprite Size Mode=::NewEnumerator1
    Sprite UV Mode=::NewEnumerator3
    Uniform Sprite Size Max=20
    Uniform Sprite Size Min=15
    UsePositionOffset=false
    Write Material Random=true
   )
   &CameraOffset(Camera Offset Amount=5)
   /SetVariables(
    Particles.RandForAttr:NiagaraInt32=&RandomRangeInt(Minimum=0, Maximum=2)
    Particles.RandInt:NiagaraInt32=&RandomRangeInt(Minimum=0, Maximum=239)
   ) {isSetParameters=true, module=SetVariables_#3}
   &self:ScratchModule_02(
    Particle Attribute Reader=DataInterface() {propertyValues {EmitterBinding {bindingMode=Other, emitterName=Strand_Chains}}}
    Particle Attribute Reader001=DataInterface() {propertyValues {EmitterBinding {bindingMode=Other, emitterName=Web_Chains}}}
    Particle Attribute Reader002=DataInterface() {propertyValues {EmitterBinding {bindingMode=Other, emitterName=Cross_Web_Chains}}}
    ParticleNum=120
    ParticleNum01=30
    ParticleNum02=90
    RandAttr:NiagaraInt32=$Particles.RandForAttr
    RandInt:NiagaraInt32=$Particles.RandInt
   ) {module=ScratchModule_02}
   &AddVelocity(
    Cone Axis=&NormalizeVector(
     Vector To Normalize=&FN_FrontEndVector3Modifier(
      Gameplay=&TransformVector(Vector=$Engine.Owner.Velocity, Source Space=::World, Destination Space=::Simulation)
      Front End=$System.FrontendVelocity
     )
     Safe Normalize=true
     Safe Fallback Threshold=0
     Safe Fallback Vector=<0 0 0.001>
    )
    Module.UseConeFalloff=true
    Module.UseVelDistribution=true
    Override Seed=false
    Velocity Mode=::NewEnumerator2
    Velocity Speed=&Multiply_Float(
     A=&RandomRangeFloat(Minimum=-100, Maximum=-200)
     B=&Lerp_Float(A=1, B=2, Alpha=$User.ParaGlideLeanAlpha)
    )
   )
  ]
  particleUpdateScript [
   &ParticleState()
   &ScaleSpriteSize(Uniform Curve Sprite Scale=DataInterface() {propertyValues {Curve {keys [
        {arriveTangent=4.6091, interpMode=RCIM_Cubic, leaveTangent=4.6091, tangentMode=RCTM_User, time=0, value=0}
        {interpMode=RCIM_Cubic, tangentMode=RCTM_User, time=0.3, value=1}
        {tangentMode=RCTM_User, time=1, value=0}
       ]}}})
   &SpriteRotationRate(Rotation Rate=&RandomRangeFloat(Minimum=300, Maximum=500))
   &Drag(Drag=&RandomRangeFloat(Minimum=1.5, Maximum=3), Use Linear Drag=true, Use Rotational Drag=true)
   &CurlNoiseForce(Noise Frequency=10, Noise Strength=250, Pan Noise=true, Pan Noise Field=<0 0 -1>)
   &SolveForcesAndVelocity(Clamp Velocity=false, Limit Acceleration=false)
   &ScaleColor(
    Scale RGB=&VectorFromFloat(Value=&FNTimeOfDayScalar(Night Scale=0.5, Morning Scale=0.75, Evening Scale=0.75))
    ScaleA=true
    ScaleRGB=true
   )
  ]
  rendererClasses [&NiagaraSpriteRendererProperties]
  renderers [{rendererClass=&NiagaraSpriteRendererProperties}]
 }
 Sparkles_Locked_Braclet {
  emitterUpdateScript [
   &EmitterState(Life Cycle Mode=::NewEnumerator1, Loop Behavior=::NewEnumerator1, Loop Duration Mode=::NewEnumerator1, UseLoopDelay=false)
   &SpawnRate(SpawnRate=100, Use Spawn Probability=false)
  ]
  particleSpawnScript [
   &InitializeParticle(
    Color=&RandomRangeLinearColor(Minimum=<rgba 10 5.4211 3.8021 0>, Maximum=<rgba 10 2.75 2.5 0.5>, Color Channel Mode=::NewEnumerator2)
    Lifetime Max=1
    Lifetime Min=0.5
    Lifetime Mode=::NewEnumerator1
    Sprite Rotation Mode=::NewEnumerator2
    Sprite Size Mode=::NewEnumerator1
    Sprite UV Mode=::NewEnumerator3
    UsePositionOffset=false
    Write Material Random=true
   )
   &CameraOffset(Camera Offset Amount=5)
   /SetVariables(
    Particles.RandFloat:NiagaraFloat=&RandomRangeFloat(Minimum=0, Maximum=1)
    Particles.RandForAttr:NiagaraInt32=&RandomRangeInt(Minimum=0, Maximum=2)
    Particles.RandInt:NiagaraInt32=&RandomRangeInt(Minimum=0, Maximum=239)
   ) {isSetParameters=true, module=SetVariables_#3}
  ]
  particleUpdateScript [
   &ParticleState()
   &self:ScratchModule_01(
    GroupNum=2
    Particle Attribute Reader=DataInterface() {propertyValues {EmitterBinding {bindingMode=Other, emitterName=Bracelet_Chains}}}
    ParticleNum=20
    RandFloat:NiagaraFloat=$Particles.RandFloat
    RandInt:NiagaraInt32=$Particles.RandInt
   ) {module=ScratchModule_01}
   &ScaleSpriteSize(Uniform Curve Sprite Scale=DataInterface() {propertyValues {Curve {keys [
        {interpMode=RCIM_Cubic, tangentMode=RCTM_User, time=0, value=0}
        {interpMode=RCIM_Cubic, tangentMode=RCTM_User, time=0.5, value=1}
        {tangentMode=RCTM_User, time=1, value=0}
       ]}}})
   &SpriteRotationRate(Rotation Rate=&RandomRangeFloat(Minimum=200, Maximum=360))
   &ScaleColor(
    Scale RGB=&VectorFromFloat(Value=&FNTimeOfDayScalar(Night Scale=0.5, Morning Scale=0.75, Evening Scale=0.75))
    ScaleA=true
    ScaleRGB=true
   )
  ]
  rendererClasses [&NiagaraSpriteRendererProperties]
  renderers [{rendererClass=&NiagaraSpriteRendererProperties}]
 }
 Sparkles_Locked_Cross_Web {
  emitterUpdateScript [
   &EmitterState(Life Cycle Mode=::NewEnumerator1, Loop Behavior=::NewEnumerator1, Loop Duration Mode=::NewEnumerator1, UseLoopDelay=false)
   &SpawnRate(SpawnRate=500, Use Spawn Probability=false)
  ]
  particleSpawnScript [
   &InitializeParticle(
    Color=&RandomRangeLinearColor(Minimum=<rgba 10 5.4211 3.8021 0>, Maximum=<rgba 10 2.75 2.5 0.3>, Color Channel Mode=::NewEnumerator2)
    Lifetime Max=1
    Lifetime Min=0.25
    Lifetime Mode=::NewEnumerator1
    Sprite Rotation Mode=::NewEnumerator2
    Sprite Size Mode=::NewEnumerator1
    Sprite UV Mode=::NewEnumerator3
    Uniform Sprite Size Max=15
    UsePositionOffset=false
    Write Material Random=true
   )
   &CameraOffset(Camera Offset Amount=5)
   /SetVariables(
    Particles.RandFloat:NiagaraFloat=&RandomRangeFloat(Minimum=0, Maximum=1)
    Particles.RandForAttr:NiagaraInt32=&RandomRangeInt(Minimum=0, Maximum=2)
    Particles.RandInt:NiagaraInt32=&RandomRangeInt(Minimum=0, Maximum=239)
   ) {isSetParameters=true, module=SetVariables_#3}
  ]
  particleUpdateScript [
   &ParticleState()
   &self:ScratchModule_01(
    GroupNum=15
    Particle Attribute Reader=DataInterface() {propertyValues {EmitterBinding {bindingMode=Other, emitterName=Cross_Web_Chains}}}
    ParticleNum=90
    RandFloat:NiagaraFloat=$Particles.RandFloat
    RandInt:NiagaraInt32=$Particles.RandInt
   ) {module=ScratchModule_01}
   &ScaleSpriteSize(Uniform Curve Sprite Scale=DataInterface() {propertyValues {Curve {keys [
        {interpMode=RCIM_Cubic, tangentMode=RCTM_User, time=0, value=0}
        {interpMode=RCIM_Cubic, tangentMode=RCTM_User, time=0.5, value=1}
        {tangentMode=RCTM_User, time=1, value=0}
       ]}}})
   &SpriteRotationRate(Rotation Rate=&RandomRangeFloat(Minimum=200, Maximum=360))
   &JitterPosition(Jitter Amount=1.5, Jitter Delay=0.2)
   &ScaleColor(
    Scale RGB=&VectorFromFloat(Value=&FNTimeOfDayScalar(Night Scale=0.5, Morning Scale=0.75, Evening Scale=0.75))
    ScaleA=true
    ScaleRGB=true
   )
  ]
  rendererClasses [&NiagaraSpriteRendererProperties]
  renderers [{rendererClass=&NiagaraSpriteRendererProperties}]
 }
 Sparkles_Locked_Strands {
  emitterUpdateScript [
   &EmitterState(Life Cycle Mode=::NewEnumerator1, Loop Behavior=::NewEnumerator1, Loop Duration Mode=::NewEnumerator1, UseLoopDelay=false)
   &SpawnRate(SpawnRate=500, Use Spawn Probability=false)
  ]
  particleSpawnScript [
   &InitializeParticle(
    Color=&RandomRangeLinearColor(Minimum=<rgba 10 5.4211 3.8021 0>, Maximum=<rgba 10 2.75 2.5 0.7>, Color Channel Mode=::NewEnumerator2)
    Lifetime Max=1
    Lifetime Min=0.5
    Lifetime Mode=::NewEnumerator1
    Sprite Rotation Mode=::NewEnumerator2
    Sprite Size Mode=::NewEnumerator1
    Sprite UV Mode=::NewEnumerator3
    Uniform Sprite Size Max=15
    UsePositionOffset=false
    Write Material Random=true
   )
   &CameraOffset(Camera Offset Amount=5)
   /SetVariables(
    Particles.RandFloat:NiagaraFloat=&RandomRangeFloat(Minimum=0, Maximum=1)
    Particles.RandForAttr:NiagaraInt32=&RandomRangeInt(Minimum=0, Maximum=2)
    Particles.RandInt:NiagaraInt32=&RandomRangeInt(Minimum=0, Maximum=239)
   ) {isSetParameters=true, module=SetVariables_#3}
  ]
  particleUpdateScript [
   &ParticleState()
   &self:ScratchModule_01(
    GroupNum=6
    Particle Attribute Reader=DataInterface() {propertyValues {EmitterBinding {bindingMode=Other, emitterName=Strand_Chains}}}
    ParticleNum=120
    RandFloat:NiagaraFloat=$Particles.RandFloat
    RandInt:NiagaraInt32=$Particles.RandInt
   ) {module=ScratchModule_01}
   &ScaleSpriteSize(Uniform Curve Sprite Scale=DataInterface() {propertyValues {Curve {keys [
        {interpMode=RCIM_Cubic, tangentMode=RCTM_User, time=0, value=0}
        {interpMode=RCIM_Cubic, tangentMode=RCTM_User, time=0.5, value=1}
        {tangentMode=RCTM_User, time=1, value=0}
       ]}}})
   &SpriteRotationRate(Rotation Rate=&RandomRangeFloat(Minimum=200, Maximum=360))
   &JitterPosition(Jitter Amount=1, Jitter Delay=0.25)
   &ScaleColor(
    Scale RGB=&VectorFromFloat(Value=&FNTimeOfDayScalar(Night Scale=0.5, Morning Scale=0.75, Evening Scale=0.75))
    ScaleA=true
    ScaleRGB=true
   )
  ]
  rendererClasses [&NiagaraSpriteRendererProperties]
  renderers [{rendererClass=&NiagaraSpriteRendererProperties}]
 }
 Sparkles_Locked_Web {
  emitterUpdateScript [
   &EmitterState(Life Cycle Mode=::NewEnumerator1, Loop Behavior=::NewEnumerator1, Loop Duration Mode=::NewEnumerator1, UseLoopDelay=false)
   &SpawnRate(SpawnRate=200, Use Spawn Probability=false)
  ]
  particleSpawnScript [
   &InitializeParticle(
    Color=&RandomRangeLinearColor(Minimum=<rgba 10 5.4211 3.8021 0>, Maximum=<rgba 10 2.75 2.5 0.5>, Color Channel Mode=::NewEnumerator2)
    Lifetime Max=1
    Lifetime Min=0.5
    Lifetime Mode=::NewEnumerator1
    Sprite Rotation Mode=::NewEnumerator2
    Sprite Size Mode=::NewEnumerator1
    Sprite UV Mode=::NewEnumerator3
    Uniform Sprite Size Max=15
    UsePositionOffset=false
    Write Material Random=true
   )
   &CameraOffset(Camera Offset Amount=5)
   /SetVariables(
    Particles.RandFloat:NiagaraFloat=&RandomRangeFloat(Minimum=0, Maximum=1)
    Particles.RandForAttr:NiagaraInt32=&RandomRangeInt(Minimum=0, Maximum=2)
    Particles.RandInt:NiagaraInt32=&RandomRangeInt(Minimum=0, Maximum=239)
   ) {isSetParameters=true, module=SetVariables_#3}
  ]
  particleUpdateScript [
   &ParticleState()
   &self:ScratchModule_01(
    GroupNum=2
    Particle Attribute Reader=DataInterface() {propertyValues {EmitterBinding {bindingMode=Other, emitterName=Web_Chains}}}
    ParticleNum=30
    RandFloat:NiagaraFloat=$Particles.RandFloat
    RandInt:NiagaraInt32=$Particles.RandInt
   ) {module=ScratchModule_01}
   &ScaleSpriteSize(Uniform Curve Sprite Scale=DataInterface() {propertyValues {Curve {keys [
        {interpMode=RCIM_Cubic, tangentMode=RCTM_User, time=0, value=0}
        {interpMode=RCIM_Cubic, tangentMode=RCTM_User, time=0.5, value=1}
        {tangentMode=RCTM_User, time=1, value=0}
       ]}}})
   &SpriteRotationRate(Rotation Rate=&RandomRangeFloat(Minimum=200, Maximum=360))
   &JitterPosition(Jitter Amount=1.5, Jitter Delay=0.25)
   &ScaleColor(
    Scale RGB=&VectorFromFloat(Value=&FNTimeOfDayScalar(Night Scale=0.5, Morning Scale=0.75, Evening Scale=0.75))
    ScaleA=true
    ScaleRGB=true
   )
  ]
  rendererClasses [&NiagaraSpriteRendererProperties]
  renderers [{rendererClass=&NiagaraSpriteRendererProperties}]
 }
 Strand_Chains {
  emitterSpawnScript [&FN_PreserveSkydiveUserVariables() {module=PreserveUserVariables}]
  emitterUpdateScript [
   &EmitterState(Life Cycle Mode=::NewEnumerator1, Loop Behavior=::NewEnumerator1, Loop Duration=1, UseLoopDelay=false)
   &SpawnBurst_Instantaneous(Spawn Count=120, Spawn Probability=HlslExpression() {hlslExpression="Emitter.LoopCount == 0"}, Use Spawn Probability=false) {module=SpawnBurst_Instantaneous001}
  ]
  particleSpawnScript [
   &FN_SetSpawnGroup(Inclusivity=true, Normalize Spawn Group=true, Particles Per Group=20, Pattern=::NewEnumerator0, Set Ribbon ID and Ribbon Link Order=true)
   &InitializeParticle(Ribbon Width=4, Ribbon Width Mode=::NewEnumerator1, UsePositionOffset=false, Write Material Random=true) {module=InitializeParticle001}
   &DynamicMaterialParameters(
    Index 0 Param 1=&SelectFloatFromArray(
     Direct Array Index=&SpawnGroup()
     Float Selection Array=DataInterface() {propertyValues {FloatData [1,3,0.5,1.68,2.3,6.4]}}
    )
    Module.Param0WriteEnabled=true
    Module.Param1WriteEnabled=true
    Module.Param2WriteEnabled=true
    Module.Param3WriteEnabled=true
   )
  ]
  particleUpdateScript [
   &ParticleState(Kill Particles When Lifetime Has Elapsed=false)
   &SkeletalMeshLocation(
    Bone / Socket Index=&ModuloInt(A=&SpawnGroup(), B=2)
    Bone Sampling Mode=::NewEnumerator4
    MaskSpawnGroup=false
    Orientation Sampling=::NewEnumerator1
    Skeletal Mesh=$User.SkeletalMesh
    Use Sampled Mesh Attribute as Ribbon ID=false
   )
   &DynamicMaterialParameters(
    Index 1 Param 1=&MakeFloatFromVector(Vector=$Particles.SkeletalMeshLocation.SampledPosition, Channel=::NewEnumerator0)
    Index 1 Param 2=&MakeFloatFromVector(Vector=$Particles.SkeletalMeshLocation.SampledPosition, Channel=::NewEnumerator1)
    Index 1 Param 3=&MakeFloatFromVector(Vector=$Particles.SkeletalMeshLocation.SampledPosition, Channel=::NewEnumerator2)
    Module.Index1Param0Write=true
    Module.Index1Param1Write=true
    Module.Index1Param2Write=true
    Module.Index1Param3Write=true
    Write Parameter Index 0=false
    Write Parameter Index 1=true
   ) {module=DynamicMaterialParameters001}
   /SetVariables(Particles.Position=&AddVectorToPosition(
     Position=$Particles.Position
     Vector=&Multiply_VectorByFloat(
      Vector=&Add_Vector(
       A=&Multiply_VectorByFloat(
        Vector=&NormalizeVector(
         Vector To Normalize=&FN_FrontEndVector3Modifier(
          Gameplay=&TransformVector(Vector=$Engine.Owner.Velocity, Source Space=::World, Destination Space=::Simulation)
          Front End=$System.FrontendVelocity
         )
         Safe Normalize=true
         Safe Fallback Threshold=0
         Safe Fallback Vector=<0 0 0.001>
        )
        Float=-1
       )
       B=&Multiply_VectorByFloat(
        Vector=&TransformVector(
         Vector=&SelectVectorFromArray(
          Direct Array Index=$Particles.SpawnGroup
          Direct Array Mode=::NewEnumerator1
          Vector Selection Array=DataInterface() {propertyValues {FloatData [<2 -2 0>, <-2 0 0>, <1 2 0>, <-0.5 0 0>, <0.5 0 0>, <-1 2 0>]}}
         )
         Destination Space=::Simulation
        )
        Float=&SelectFloatFromArray(
         Direct Array Index=&SpawnGroup()
         Float Selection Array=DataInterface() {propertyValues {FloatData [0.1,0.15,0.08,0.16,0.15,0.18]}}
        )
       )
      )
      Float=&FloatFromCurve(
       FloatCurve=DataInterface() {propertyValues {Curve {keys [{tangentMode=RCTM_User, time=0, value=0}, {tangentMode=RCTM_User, time=1, value=1}]}}}
       CurveIndex=$Particles.Normalized Spawn Group
       Scale Curve=&SelectFloatFromArray(
        Direct Array Index=&SpawnGroup()
        Float Selection Array=DataInterface() {propertyValues {FloatData [105,115,120,110,105,100]}}
       )
      )
     )
    )) {isSetParameters=true}
   /SetVariables(Particles.Position=&AddVectorToPosition(
     Position=$Particles.Position
     Vector=&Multiply_VectorByFloat(
      Vector=&RotateVector(
       Vector To Rotate:Vector3f=<10 0 3>
       Yaw:NiagaraFloat=&Add_Float(
        A=$Particles.Age
        B=&Multiply_Float(A=&MakeFloatFromInt(Int=&SpawnGroup()), B=2)
       )
       Pitch:NiagaraFloat=0
       Roll:NiagaraFloat=0
       Rotation Rate:NiagaraFloat=&SelectFloatFromArray(
        Direct Array Index=&SpawnGroup()
        Float Selection Array=DataInterface() {propertyValues {FloatData [150,140,130,120,160,150]}}
       )
      )
      Float=&Multiply_Float(
       A=&Sine(
        Normalized Angle=&Add_Float(
         A=&Multiply_Float(
          A=$Emitter.Age
          B=&SelectFloatFromArray(
           Direct Array Index=&SpawnGroup()
           Float Selection Array=DataInterface() {propertyValues {FloatData [-2,-2.5,-2.5,-3,-3,-2]}}
          )
         )
         B=$Particles.Normalized Spawn Group
        )
        Period=0.9
        Scale=1
        Bias=0
       )
       B=&FloatFromCurve(
        FloatCurve=DataInterface() {propertyValues {Curve {keys [{tangentMode=RCTM_User, time=0, value=0}, {tangentMode=RCTM_User, time=1, value=1}]}}}
        CurveIndex=$Particles.Normalized Spawn Group
       )
      )
     )
    )) {isSetParameters=true}
   &FN_Sprite_Attributes("Create Normalized Age "=true)
   &ScaleColor(
    Scale RGB=&VectorFromFloat(Value=&FNTimeOfDayScalar(Night Scale=0.8, Morning Scale=0.9, Evening Scale=0.9))
    ScaleA=true
    ScaleRGB=true
   )
  ]
  rendererClasses [&NiagaraRibbonRendererProperties]
  renderers [{rendererClass=&NiagaraRibbonRendererProperties}]
 }
 Web_Chains {
  emitterSpawnScript [&FN_PreserveSkydiveUserVariables()]
  emitterUpdateScript [
   &EmitterState(
    Inactive Response=::NewEnumerator1
    Life Cycle Mode=::NewEnumerator1
    Loop Behavior=::NewEnumerator1
    Loop Duration Mode=::NewEnumerator1
    UseLoopDelay=false
   )
   &SpawnBurst_Instantaneous(Spawn Count=30, Use Spawn Probability=false, UseLoopCountLimit=false)
   &SampleSkeletalMesh(
    AdditionalYawPitchRoll=false
    Bone Sampling Mode:ENiagaraBoneSamplingMode=::NewEnumerator4
    Skeletal Mesh:NiagaraDataInterfaceSkeletalMesh=$User.SkeletalMesh
   )
   &SampleSkeletalMesh(
    AdditionalYawPitchRoll=false
    Bone / Socket Index=1
    Bone Sampling Mode:ENiagaraBoneSamplingMode=::NewEnumerator4
    Skeletal Mesh:NiagaraDataInterfaceSkeletalMesh=$User.SkeletalMesh
   ) {module=SampleSkeletalMesh001}
  ]
  particleSpawnScript [
   &FN_SetSpawnGroup(Inclusivity=true, Normalize Spawn Group=true, Particles Per Group=15, Pattern=::NewEnumerator0, Set Ribbon ID and Ribbon Link Order=true)
   &InitializeRibbon(Ribbon Width=3, UsePositionOffset=false, Write Material Random=true)
   &DynamicMaterialParameters(Module.Param0WriteEnabled=true, Module.Param1WriteEnabled=true, Module.Param2WriteEnabled=false, Module.Param3WriteEnabled=false)
  ]
  particleUpdateScript [
   &ParticleState(Kill Particles When Lifetime Has Elapsed=false, Let Infinitely Lived Particles Die When Emitter Deactivates=true)
   /SetVariables(Particles.float:NiagaraFloat=&FloatFromCurve(FloatCurve=DataInterface() {propertyValues {Curve {keys [<at 0.6 0 RCIM_Cubic>,<at 1 1 RCIM_Cubic>]}}}, CurveIndex=$User.ParaGlideLeanAlpha)) {isSetParameters=true}
   /Game/Effects/Niagara/Modules/Curves/FN_HermiteSpline(
    Direction End:Vector3f=&Add_Vector(
     A=&Add_Vector(
      A=&MaskVectorBySpawnGroup(
       Masked Vector=&Multiply_VectorByFloat(Vector=<-0.1 -0.2 0.25>, Float=150)
       Passthrough Vector=&Multiply_VectorByFloat(Vector=<-0.1 0.2 0.6>, Float=200)
      )
      B=&Multiply_VectorByFloat(
       Vector=&NormalizeVector(
        Vector To Normalize=&FN_FrontEndVector3Modifier(
         Gameplay=&TransformVector(Vector=$Engine.Owner.Velocity, Source Space=::World, Destination Space=::Simulation)
         Front End=$System.FrontendVelocity
        )
        Safe Normalize=true
        Safe Fallback Threshold=0
        Safe Fallback Vector=<0 0 0.001>
       )
       Float=&Lerp_Float(A=-100, B=-150, Alpha=$User.ParaGlideLeanAlpha)
      )
     )
     B=&Multiply_VectorByFloat(
      Vector=&NormalizeVector(
       Vector To Normalize=&Lerp_Vector(A=<0 0 1>, B=<0 1 1>, Alpha=$Particles.float)
       Safe Normalize=true
       Safe Fallback Threshold=0
       Safe Fallback Vector=<0 0 0.001>
      )
      Float=&Lerp_Float(A=0, B=&FN_FrontEndFloatModifier(Gameplay=100, Front End=0), Alpha=$Particles.float)
     )
    )
    Direction Start:Vector3f=&Add_Vector(
     A=&Add_Vector(
      A=&MaskVectorBySpawnGroup(
       Masked Vector=&Multiply_VectorByFloat(Vector=<0.1 -0.2 0.25>, Float=150)
       Passthrough Vector=&Multiply_VectorByFloat(Vector=<0.1 0.2 0.6>, Float=200)
      )
      B=&Multiply_VectorByFloat(
       Vector=&NormalizeVector(
        Vector To Normalize=&FN_FrontEndVector3Modifier(
         Gameplay=&TransformVector(Vector=$Engine.Owner.Velocity, Source Space=::World, Destination Space=::Simulation)
         Front End=$System.FrontendVelocity
        )
        Safe Normalize=true
        Safe Fallback Threshold=0
        Safe Fallback Vector=<0 0 0.001>
       )
       Float=&Lerp_Float(A=-100, B=-150, Alpha=$User.ParaGlideLeanAlpha)
      )
     )
     B=&Multiply_VectorByFloat(
      Vector=&NormalizeVector(
       Vector To Normalize=&Lerp_Vector(A=<0 0 1>, B=<0 1 1>, Alpha=$Particles.float)
       Safe Normalize=true
       Safe Fallback Threshold=0
       Safe Fallback Vector=<0 0 0.001>
      )
      Float=&Lerp_Float(A=0, B=&FN_FrontEndFloatModifier(Gameplay=100, Front End=0), Alpha=$Particles.float)
     )
    )
    Position Blend:NiagaraFloat=$Particles.Normalized Spawn Group
    Position End:Vector3f=&AddVectorToPosition(
     Position=$Emitter.SampleSkeletalMesh001.SampledPosition
     Vector=&RotateVectorByQuaternion(Vector To Rotate=<-2 -2 0>, Quaternion=$Emitter.SampleSkeletalMesh001.SampledRotation)
    )
    Position Start:Vector3f=&AddVectorToPosition(
     Position=Linked() {linkedVariable {name=Emitter.SampleSkeletalMesh.SampledPosition, type=Vector3f}}
     Vector=&RotateVectorByQuaternion(Vector To Rotate=<2 2 0>, Quaternion=$Emitter.SampleSkeletalMesh.SampledRotation)
    )
    Write Particles Position=true
   )
   &ScaleRibbonWidth(Ribbon Width Scale=&FloatFromCurve(
     FloatCurve=DataInterface() {propertyValues {Curve {keys [
         {interpMode=RCIM_Cubic, tangentMode=RCTM_User, time=0, value=0.7}
         {interpMode=RCIM_Cubic, tangentMode=RCTM_User, time=0.5, value=1}
         {tangentMode=RCTM_User, time=1, value=0.7}
        ]}}}
     CurveIndex=$Particles.Normalized Spawn Group
    ))
   &PendulumSetup(
    Module.Enable Spring Driver=false
    Pendulum Length:NiagaraFloat=&FloatFromCurve(
     FloatCurve=DataInterface() {propertyValues {Curve {keys [
         {arriveTangent=2.8666, interpMode=RCIM_Cubic, leaveTangent=2.8666, tangentMode=RCTM_User, time=0, value=0}
         {interpMode=RCIM_Cubic, tangentMode=RCTM_User, time=0.5, value=1}
         {arriveTangent=-3.7423, interpMode=RCIM_Cubic, leaveTangent=-3.7423, tangentMode=RCTM_User, time=1, value=0}
        ]}}}
     CurveIndex=$Particles.Normalized Spawn Group
    )
    Pendulum Pivot:NiagaraPosition=$Particles.Position
    Rigid Pendulum=false
   )
   &CurlNoiseForce(Noise Strength=&Lerp_Float(A=500, B=1500, Alpha=$System.Dive), Pan Noise=true, Pan Noise Field=<0 0 10>)
   &SolveForcesAndVelocity(Clamp Velocity=false, Limit Acceleration=false)
   &PendulumConstraint(Calculate and Output Potential Energy=true)
   &CalculateAccurateVelocity()
   &ScaleColor(
    Scale RGB=&VectorFromFloat(Value=&FNTimeOfDayScalar(Night Scale=0.8, Morning Scale=0.9, Evening Scale=0.9))
    ScaleA=true
    ScaleRGB=true
   ) {module=ScaleColor001}
  ]
  rendererClasses [&NiagaraRibbonRendererProperties]
  renderers [{rendererClass=&NiagaraRibbonRendererProperties}]
 }
}
summary {
 systemName=NS_Contrail_BlazerVeil
 emitters [
  {emitterName=Web_Chains, rendererClasses [&NiagaraRibbonRendererProperties]}
  {emitterName=Cross_Web_Chains, rendererClasses [&NiagaraRibbonRendererProperties]}
  {
   emitterName=Pearls_On_Web_Chains
   rendererClasses [&NiagaraMeshRendererProperties, &NiagaraSpriteRendererProperties]
  }
  {
   emitterName=Pearls_On_Strand_Chains
   rendererClasses [&NiagaraMeshRendererProperties, &NiagaraSpriteRendererProperties]
  }
  {emitterName=Strand_Chains, rendererClasses [&NiagaraRibbonRendererProperties]}
  {emitterName=Bracelet_Chains, rendererClasses [&NiagaraRibbonRendererProperties]}
  {
   emitterName=Diamonds_On_Strand_Chains
   rendererClasses [&NiagaraMeshRendererProperties, &NiagaraSpriteRendererProperties]
  }
  {
   emitterName=Hand_And_Center_Diamond
   rendererClasses [&NiagaraMeshRendererProperties, &NiagaraSpriteRendererProperties]
  }
  {emitterName=Sparkles_Locked_Strands, rendererClasses [&NiagaraSpriteRendererProperties]}
  {emitterName=Sparkles_Everywhere_Flying, rendererClasses [&NiagaraSpriteRendererProperties]}
  {emitterName=Sparkles_Locked_Braclet, rendererClasses [&NiagaraSpriteRendererProperties]}
  {emitterName=Sparkles_Locked_Web, rendererClasses [&NiagaraSpriteRendererProperties]}
  {emitterName=Sparkles_Locked_Cross_Web, rendererClasses [&NiagaraSpriteRendererProperties]}
 ]
 userVariables [
  User.bIsFrontEnd {type=NiagaraBool}
  User.bIsFrontEndPreview {type=NiagaraBool}
  User.GradientCenterpointSparksFlat {type=NiagaraFloat, defaultValue=0.5, description=""}
  User.GradientCenterpointSparksSide {type=NiagaraFloat, defaultValue=0.9, description=""}
  User.ParaGlideLeanAlpha {type=NiagaraFloat, defaultValue=0, description=""}
  User.PearlColor {type=LinearColor, description="", defaultValue {struct=LinearColor, value=<rgba 1.5 1.15 0.9 1>}}
  User.PearlColorFrontend {type=LinearColor, description="", defaultValue {struct=LinearColor, value=<rgba 1 0.9 0.8 1>}}
  User.PlayerSpeed {type=NiagaraFloat, defaultValue=0, description=""}
  User.RibbonIDOffset {type=NiagaraInt32}
  User.SkeletalMesh {
   type=NiagaraDataInterfaceSkeletalMesh
   defaultValue {
    struct=NiagaraExt_VariableValue_DataInterface
    value {dataInterface=&self:NiagaraDataInterfaceSkeletalMesh_0}
   }
  }
 ]
}
